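# Animation Blueprint for the DiscoButler glider, animating deploy, idle and closed states with frontend variants.
# A Parachute state machine follows skydiving and force-open state, with notifies toggling the handle bone and marking full deployment.
# Curve-weighted spin and jitter additives layer on top, with spin rate following forward acceleration.
# Outside frontend, each update interpolates ten canopy flap rolls from acceleration and bone orientation, applied through Modify Bone nodes.
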AnimBlueprint /BRCosmetics/Animation/Game/MainPlayer/Skydive/ParaGlide/DiscoButler/DiscoButler_Glider_AnimBP
target_skeleton=/BRCosmetics/Gadgets/Assets/VinderTech_GliderChute/Glider_DiscoButler/Meshes/DiscoButler_Glider_Skeleton
graphs [
 AnimGraph {nodes [
   Apply Additive!(Base, Additive) -> (Pose) {alphaBoolBlend {blendOption=Linear}}
   n1 = "Sequence Player (additive) 'DiscoButler_TopSpin_GLIDER'"!(PlayRate:float=Get TopSpin_Speed() -> (TopSpin_Speed:double)) -> (Pose) {bStartFromMatchingPose=true, sequence=&/DiscoButler_TopSpin_GLIDER}
   AnimGraphNode_StateMachine_2 = Parachute() {is=title AnimGraphNode_StateMachine}
   n2 = Apply Additive!(Base=@AnimGraphNode_StateMachine_2, Additive=@n1, AlphaCurveName:Name=TopSpinAlpha) -> (Pose) {alphaInputType=Curve, alphaBoolBlend {blendOption=Linear}, alphaScaleBiasClamp {bias=1, scale=-1}}
   n3 = "Sequence Player (additive) 'DiscoButler_Jitter_GLIDER'"() {sequence=&/DiscoButler_Jitter_GLIDER}
   n4 = "Transform (Modify) Bone - Bone: JNT_CanopyFlap_A"(
    ComponentPose=Local To Component(LocalPose=Apply Additive(Base=@n2, Additive=@n3, AlphaCurveName=jitter_DiscoButler) {alphaInputType=Curve, alphaBoolBlend {blendOption=Linear}, alphaScaleBiasClamp {bias=1, scale=-1}})
    Alpha=1
    Translation="0,0,0"
    Rotation="Get (a ref)"(Array:Array of Rotators=Get FlapRotationValues() -> (FlapRotationValues:Array of Rotators), Dimension 1=0) -> (Output:"Rotator (by ref)")
    Scale="1,1,1"
   ) {
    rotationMode=BMM_Additive
    rotationSpace=BCS_BoneSpace
    scaleMode=BMM_Additive
    scaleSpace=BCS_BoneSpace
    boneToModify {boneName=JNT_CanopyFlap_A}
   }
   n5 = "Transform (Modify) Bone - Bone: JNT_CanopyFlap_B"(ComponentPose=@n4, Alpha=1, Translation="0,0,0", Rotation="Get (a ref)"(Array=Get FlapRotationValues(), Dimension 1=1), Scale="1,1,1") {
    rotationMode=BMM_Additive
    rotationSpace=BCS_BoneSpace
    scaleMode=BMM_Additive
    scaleSpace=BCS_BoneSpace
    boneToModify {boneName=JNT_CanopyFlap_B}
   }
   n6 = "Transform (Modify) Bone - Bone: JNT_CanopyFlap_C"(ComponentPose=@n5, Alpha=1, Translation="0,0,0", Rotation="Get (a ref)"(Array=Get FlapRotationValues(), Dimension 1=2), Scale="1,1,1") {
    rotationMode=BMM_Additive
    rotationSpace=BCS_BoneSpace
    scaleMode=BMM_Additive
    scaleSpace=BCS_BoneSpace
    boneToModify {boneName=JNT_CanopyFlap_C}
   }
   n7 = "Transform (Modify) Bone - Bone: JNT_CanopyFlap_I"(ComponentPose=@n6, Alpha=1, Translation="0,0,0", Rotation="Get (a ref)"(Array=Get FlapRotationValues(), Dimension 1=8), Scale="1,1,1") {
    rotationMode=BMM_Additive
    rotationSpace=BCS_BoneSpace
    scaleMode=BMM_Additive
    scaleSpace=BCS_BoneSpace
    boneToModify {boneName=JNT_CanopyFlap_I}
   }
   n8 = "Transform (Modify) Bone - Bone: JNT_CanopyFlap_J"(ComponentPose=@n7, Alpha=1, Translation="0,0,0", Rotation="Get (a ref)"(Array=Get FlapRotationValues(), Dimension 1=9), Scale="1,1,1") {
    rotationMode=BMM_Additive
    rotationSpace=BCS_BoneSpace
    scaleMode=BMM_Replace
    scaleSpace=BCS_BoneSpace
    boneToModify {boneName=JNT_CanopyFlap_J}
   }
   n9 = "Transform (Modify) Bone - Bone: JNT_CanopyFlap_F"(ComponentPose=@n8, Alpha=1, Translation="0,0,0", Rotation="Get (a ref)"(Array=Get FlapRotationValues(), Dimension 1=5), Scale="1,1,1") {
    rotationMode=BMM_Additive
    rotationSpace=BCS_BoneSpace
    scaleMode=BMM_Additive
    scaleSpace=BCS_BoneSpace
    boneToModify {boneName=JNT_CanopyFlap_F}
   }
   n10 = "Transform (Modify) Bone - Bone: JNT_CanopyFlap_G"(ComponentPose=@n9, Alpha=1, Translation="0,0,0", Rotation="Get (a ref)"(Array=Get FlapRotationValues(), Dimension 1=6), Scale="1,1,1") {
    rotationMode=BMM_Additive
    rotationSpace=BCS_BoneSpace
    scaleMode=BMM_Additive
    scaleSpace=BCS_BoneSpace
    boneToModify {boneName=JNT_CanopyFlap_G}
   }
   n11 = "Transform (Modify) Bone - Bone: JNT_CanopyFlap_H"(ComponentPose=@n10, Alpha=1, Translation="0,0,0", Rotation="Get (a ref)"(Array=Get FlapRotationValues(), Dimension 1=7), Scale="1,1,1") {
    rotationMode=BMM_Additive
    rotationSpace=BCS_BoneSpace
    scaleMode=BMM_Additive
    scaleSpace=BCS_BoneSpace
    boneToModify {boneName=JNT_CanopyFlap_H}
   }
   n12 = "Transform (Modify) Bone - Bone: JNT_CanopyFlap_D"(ComponentPose=@n11, Alpha=1, Translation="0,0,0", Rotation="Get (a ref)"(Array=Get FlapRotationValues(), Dimension 1=3), Scale="1,1,1") {
    rotationMode=BMM_Additive
    rotationSpace=BCS_BoneSpace
    scaleMode=BMM_Additive
    scaleSpace=BCS_BoneSpace
    boneToModify {boneName=JNT_CanopyFlap_D}
   }
   n13 = "Transform (Modify) Bone - Bone: JNT_CanopyFlap_E"(ComponentPose=@n12, Alpha=1, Translation="0,0,0", Rotation="Get (a ref)"(Array=Get FlapRotationValues(), Dimension 1=4), Scale="1,1,1") {
    rotationMode=BMM_Additive
    rotationSpace=BCS_BoneSpace
    scaleMode=BMM_Additive
    scaleSpace=BCS_BoneSpace
    boneToModify {boneName=JNT_CanopyFlap_E}
   }
   Output Pose(Result=Slot 'DefaultSlot'(Source=Component To Local(ComponentPose=@n13))) {layerGroup=None}
  ]}
 Parachute {
  entry=Closed
  path=AnimGraph.AnimGraphNode_StateMachine_2.Parachute
  ids=id
  nodes [
   AnimStateNode_0 = Parachute_Into() {is=title AnimStateNode, stateEntered {guid=#1, notifyName=DeployEntered}}
   AnimStateNode_1 = Idle()
   AnimStateNode_2 = Closed() {is=title AnimStateNode, stateEntered {guid=#2, notifyName=ClosedEntered}}
   AnimStateTransitionNode_0 = AnimStateTransitionNode() {bAutomaticRuleBasedOnSequencePlayerInState=true, crossfadeDuration=0}
   AnimStateTransitionNode_1 = AnimStateTransitionNode()
   AnimStateTransitionNode_2 = AnimStateTransitionNode() {crossfadeDuration=0}
   AnimStateTransitionNode_3 = AnimStateTransitionNode() {crossfadeDuration=0}
  ]
  links [
   {
    dst=Parachute_Into
    rule=AnimGraph.AnimGraphNode_StateMachine_2.Parachute.AnimStateTransitionNode_1.AnimationTransitionGraph_1
    src=Closed
    via=AnimStateTransitionNode_1
   }
   {
    dst=Closed
    rule=AnimGraph.AnimGraphNode_StateMachine_2.Parachute.AnimStateTransitionNode_2.AnimationTransitionGraph_1
    src=Idle
    via=AnimStateTransitionNode_2
   }
   {
    dst=Closed
    rule=AnimGraph.AnimGraphNode_StateMachine_2.Parachute.AnimStateTransitionNode_3.AnimationTransitionGraph_1
    src=Parachute_Into
    via=AnimStateTransitionNode_3
   }
   {
    dst=Idle
    rule=AnimGraph.AnimGraphNode_StateMachine_2.Parachute.AnimStateTransitionNode_0.AnimationTransitionGraph_1
    src=Parachute_Into
    via=AnimStateTransitionNode_0
   }
  ]
 }
 Parachute_Into {
  path=AnimGraph.AnimGraphNode_StateMachine_2.Parachute.AnimStateNode_0.Parachute_Into
  nodes [
   n1 = Sequence Player 'DiscoButler_FrontEnd_Into_GLIDER'() {bLoopAnimation=false, sequence=&/DiscoButler_FrontEnd_Into_GLIDER}
   Output Animation Pose(Result=Blend Poses by bool(
     BlendPose_0=@n1
     BlendPose_1=Sequence Player 'DiscoButler_Into_GLIDER'!(PlayRate:float=1) -> (Pose) {bLoopAnimation=false, sequence=&/DiscoButler_Into_GLIDER}
     bActiveValue=Get IsFrontend() -> (IsFrontend:bool)
    ))
  ]
 }
 Idle {
  path=AnimGraph.AnimGraphNode_StateMachine_2.Parachute.AnimStateNode_1.Idle
  nodes [Output Animation Pose(Result=Blend Poses by bool(
     BlendPose_0=Sequence Player 'DiscoButler_FrontEnd_GLIDER'() {sequence=&/DiscoButler_FrontEnd_GLIDER}
     BlendPose_1=Sequence Player 'DiscoButler_Base_Pose_GLIDER'!(PlayRate:float=1) -> (Pose) {sequence=&/DiscoButler_Base_Pose_GLIDER}
     bActiveValue=Get IsFrontend()
    ))]
 }
 Closed {
  path=AnimGraph.AnimGraphNode_StateMachine_2.Parachute.AnimStateNode_2.Closed
  nodes [
   n1 = Sequence Player 'DiscoButler_FrontEnd_Close_GLIDER'() {bLoopAnimation=false, sequence=&/DiscoButler_FrontEnd_Close_GLIDER}
   Output Animation Pose(Result=Blend Poses by bool(
     BlendPose_0=@n1
     BlendPose_1=Sequence Player 'DiscoButler_Closed_GLIDER'() {sequence=&/DiscoButler_Closed_GLIDER}
     bActiveValue=Get IsFrontend()
    ))
  ]
 }
 AnimationTransitionGraph_1 {
  path=AnimGraph.AnimGraphNode_StateMachine_2.Parachute.AnimStateTransitionNode_0.AnimationTransitionGraph_1
  nodes [
   Result()
   "float < float"(A="Get Relevant Anim Time Remaining (Parachute_Into)"!() -> (ReturnValue:float) {is=title K2Node_AnimGetter}, B=3)
  ]
 }
 AnimationTransitionGraph_1 {
  path=AnimGraph.AnimGraphNode_StateMachine_2.Parachute.AnimStateTransitionNode_1.AnimationTransitionGraph_1
  nodes [
   n1 = Get SkydivingState() -> (SkydivingState:Fort Anim Input Skydiving Structure)
   Result(bCanEnterTransition=OR Boolean(
     A=Break Fort Anim Input Skydiving!(FortAnimInput_Skydiving:"Fort Anim Input Skydiving Structure (by ref)"=@n1) -> (
      ParaGilderRootOffset:Vector=~
      ParaGliderRootRotationOverride:Rotator=~
      bUseParaGlideRootModifier:bool=~
      bIsSkydivingFromLaunchPad:bool=~
      bIsSkydivingFromBus:bool=~
      bIsInVortex:bool=~
      bIsDiving:bool=~
      bIsParachuteOpen:bool=~
      bIsSkydiving:bool=~
      bIsParachuteLeaning:bool=~
      bIsSkydiveLeaning:bool=~
      bIsLeaning:bool=~
      bIsSkydiveDiveMode:bool=~
      bPlaySkydiveDrift:bool=~
      bSkydiveDriftDelayActive:bool=~
      bSkydiveDriftAnimAllowed:bool=~
      bIsGliderRight:bool=~
      bIsGliderCenter:bool=~
      bIsGliderLeft:bool=~
      bIsGliderForward:bool=~
      bIsGliderBack:bool=~
      bHasDeployedAndIsStandingUpright:bool=~
      LocalAccelForward:float=~
      LocalAccelRight:float=~
      LocalVelocityRight:float=~
      SkydiveAimPitch:float=~
      SkydiveAimYaw:float=~
      DeployChuteAnimRate:float=~
      SkydiveDriftAnimRate:float=~
      SkydiveDriftAnimRateCurrent:float=~
      SkydiveFidgetAnimRate:float=~
      SkydiveFidgetAnimRateCurrent:float=~
      SkydiveAdditiveAlpha:float=~
      SkydiveDriftDelay:float=~
      SkydiveDriftAnim:int
      SkydiveDriftAnimMax:int
      LaunchpadAnim:int
      LocalAccelDir:ESkydivingDirection Enum
      DirectionLast:ESkydivingDirection Enum
      PoseAdditive:Anim Sequence Object Reference
     ) {is=title K2Node_BreakStruct}.bIsParachuteOpen
     B=Get bForceOpen() -> (bForceOpen:bool)
    ))
   n2 = Break Fort Anim Input Skydiving(FortAnimInput_Skydiving=Get SkydivingState()) -> (
    ParaGilderRootOffset=~
    ParaGliderRootRotationOverride=~
    bUseParaGlideRootModifier=~
    bIsSkydivingFromLaunchPad=~
    bIsSkydivingFromBus=~
    bIsInVortex=~
    bIsDiving=~
    bIsParachuteOpen=~
    bIsSkydiving=~
    bIsParachuteLeaning=~
    bIsSkydiveLeaning=~
    bIsLeaning=~
    bIsSkydiveDiveMode=~
    bPlaySkydiveDrift=~
    bSkydiveDriftDelayActive=~
    bSkydiveDriftAnimAllowed=~
    bIsGliderRight=~
    bIsGliderCenter=~
    bIsGliderLeft=~
    bIsGliderForward=~
    bIsGliderBack=~
    bHasDeployedAndIsStandingUpright=~
    LocalAccelForward=~
    LocalAccelRight=~
    LocalVelocityRight=~
    SkydiveAimPitch=~
    SkydiveAimYaw=~
    DeployChuteAnimRate=~
    SkydiveDriftAnimRate=~
    SkydiveDriftAnimRateCurrent=~
    SkydiveFidgetAnimRate=~
    SkydiveFidgetAnimRateCurrent=~
    SkydiveAdditiveAlpha=~
    SkydiveDriftDelay=~
   )
   n3 = Get OwnerGlider() -> (OwnerGlider:Fort Player Parachute Object Reference)
   AND Boolean(
    A=OR Boolean!(
     A=@n2.bIsSkydivingFromLaunchPad
     B=@n2.bIsSkydivingFromBus
     C:bool=@n2.bIsInVortex
     D:bool=Get IsCosmeticPreview!(self:Fort Player Parachute Object Reference=@n3) -> (IsCosmeticPreview:bool)
    ) -> (ReturnValue)
    B=@n2.bIsParachuteOpen
   )
  ]
 }
 AnimationTransitionGraph_1 {
  path=AnimGraph.AnimGraphNode_StateMachine_2.Parachute.AnimStateTransitionNode_2.AnimationTransitionGraph_1
  nodes [
   n1 = Get bForceOpen()
   n2 = Break Fort Anim Input Skydiving(FortAnimInput_Skydiving=Get SkydivingState()) -> (
    ParaGilderRootOffset=~
    ParaGliderRootRotationOverride=~
    bUseParaGlideRootModifier=~
    bIsSkydivingFromLaunchPad=~
    bIsSkydivingFromBus=~
    bIsInVortex=~
    bIsDiving=~
    bIsParachuteOpen=~
    bIsSkydiving=~
    bIsParachuteLeaning=~
    bIsSkydiveLeaning=~
    bIsLeaning=~
    bIsSkydiveDiveMode=~
    bPlaySkydiveDrift=~
    bSkydiveDriftDelayActive=~
    bSkydiveDriftAnimAllowed=~
    bIsGliderRight=~
    bIsGliderCenter=~
    bIsGliderLeft=~
    bIsGliderForward=~
    bIsGliderBack=~
    bHasDeployedAndIsStandingUpright=~
    LocalAccelForward=~
    LocalAccelRight=~
    LocalVelocityRight=~
    SkydiveAimPitch=~
    SkydiveAimYaw=~
    DeployChuteAnimRate=~
    SkydiveDriftAnimRate=~
    SkydiveDriftAnimRateCurrent=~
    SkydiveFidgetAnimRate=~
    SkydiveFidgetAnimRateCurrent=~
    SkydiveAdditiveAlpha=~
    SkydiveDriftDelay=~
   )
   Result(bCanEnterTransition=AND Boolean(A=NOT Boolean(A=@n2.bIsParachuteOpen), B=NOT Boolean(A=@n1)))
  ]
 }
 AnimationTransitionGraph_1 {
  path=AnimGraph.AnimGraphNode_StateMachine_2.Parachute.AnimStateTransitionNode_3.AnimationTransitionGraph_1
  nodes [
   n1 = Break Fort Anim Input Skydiving(FortAnimInput_Skydiving=Get SkydivingState()) -> (
    ParaGilderRootOffset=~
    ParaGliderRootRotationOverride=~
    bUseParaGlideRootModifier=~
    bIsSkydivingFromLaunchPad=~
    bIsSkydivingFromBus=~
    bIsInVortex=~
    bIsDiving=~
    bIsParachuteOpen=~
    bIsSkydiving=~
    bIsParachuteLeaning=~
    bIsSkydiveLeaning=~
    bIsLeaning=~
    bIsSkydiveDiveMode=~
    bPlaySkydiveDrift=~
    bSkydiveDriftDelayActive=~
    bSkydiveDriftAnimAllowed=~
    bIsGliderRight=~
    bIsGliderCenter=~
    bIsGliderLeft=~
    bIsGliderForward=~
    bIsGliderBack=~
    bHasDeployedAndIsStandingUpright=~
    LocalAccelForward=~
    LocalAccelRight=~
    LocalVelocityRight=~
    SkydiveAimPitch=~
    SkydiveAimYaw=~
    DeployChuteAnimRate=~
    SkydiveDriftAnimRate=~
    SkydiveDriftAnimRateCurrent=~
    SkydiveFidgetAnimRate=~
    SkydiveFidgetAnimRateCurrent=~
    SkydiveAdditiveAlpha=~
    SkydiveDriftDelay=~
   )
   n2 = Get bForceOpen()
   Result(bCanEnterTransition=AND Boolean(A=NOT Boolean(A=@n1.bIsParachuteOpen), B=NOT Boolean(A=@n2)))
  ]
 }
 CalculateFlapRotation {nodes [
   n1 = CalculateFlapRotation!() -> (then, FlapToCalculate:Name, CurrentRotation:Rotator=~) {is=title K2Node_FunctionEntry}
   n2 = Break Fort Anim Input Skydiving(FortAnimInput_Skydiving=Get SkydivingState()) -> (
    ParaGilderRootOffset=~
    ParaGliderRootRotationOverride=~
    bUseParaGlideRootModifier=~
    bIsSkydivingFromLaunchPad=~
    bIsSkydivingFromBus=~
    bIsInVortex=~
    bIsDiving=~
    bIsParachuteOpen=~
    bIsSkydiving=~
    bIsParachuteLeaning=~
    bIsSkydiveLeaning=~
    bIsLeaning=~
    bIsSkydiveDiveMode=~
    bPlaySkydiveDrift=~
    bSkydiveDriftDelayActive=~
    bSkydiveDriftAnimAllowed=~
    bIsGliderRight=~
    bIsGliderCenter=~
    bIsGliderLeft=~
    bIsGliderForward=~
    bIsGliderBack=~
    bHasDeployedAndIsStandingUpright=~
    LocalAccelForward=~
    LocalAccelRight=~
    LocalVelocityRight=~
    SkydiveAimPitch=~
    SkydiveAimYaw=~
    DeployChuteAnimRate=~
    SkydiveDriftAnimRate=~
    SkydiveDriftAnimRateCurrent=~
    SkydiveFidgetAnimRate=~
    SkydiveFidgetAnimRateCurrent=~
    SkydiveAdditiveAlpha=~
    SkydiveDriftDelay=~
   )
   n3 = "Clamp (Float)"(Value=@n2.LocalAccelRight, Min=-1, Max=1)
   n4 = Get OwningComponent() -> (OwningComponent:Skeletal Mesh Component Object Reference)
   n5 = GetSocketTransform!(self=@n4, InSocketName=@n1, TransformSpace=RTS_Component) -> (ReturnValue_Location:Vector, ReturnValue_Rotation:Rotator, ReturnValue_Scale:Vector)
   n6 = GetSocketTransform(self=@n4, InSocketName=Melee_TwoHanded_Handle, TransformSpace=RTS_Component)
   n7 = Get FlapMaxDynamicAngle() -> (FlapMaxDynamicAngle:double)
   n8 = float * float(
    A=Dot Product!(
     A:Vector=Get Rotation X Vector!(InRot:Rotator=@n5.ReturnValue_Rotation) -> (ReturnValue:Vector) {is=title K2Node_CallFunction}
     B:Vector=Get Rotation X Vector(InRot=@n6.ReturnValue_Rotation)
    ) -> (ReturnValue:double)
    B=@n7
   ) {is=title K2Node_PromotableOperator}
   n9 = "Clamp (Float)"(Value=@n2.LocalAccelForward, Min=-1, Max=1)
   n10 = float * float(A=@n9, B=@n7)
   n11 = "Clamp (Float)"(Value=float + float(A=float * float(A=@n3, B=@n8), B=float * float(A=@n9, B=@n10)), Min=-90, Max=90)
   n12 = float + float(A=float + float(A="Absolute (Float)"(A=@n3), B="Absolute (Float)"(A=@n9)), B=1)
   n13 = Get FlapDynamicsInterpolationSpeed() -> (FlapDynamicsInterpolationSpeed:double)
   > Return Node!(
    execute
    NewRotation:Rotator=RInterpTo!(
     Current:Rotator=@n1.CurrentRotation
     Target:Rotator=MakeRotator(Roll=@n11)
     DeltaTime:float=Get Time() -> (Time:double)
     InterpSpeed:float=float * float(A=@n12, B=@n13)
    ) -> (ReturnValue:Rotator)
   )
  ]}
 EventGraph {nodes [
   n1 = Event BlueprintUpdateAnimation()
   > Set Time(Time:double=@n1.DeltaTimeX) -> (Output_Get:double)
   > Get Player Open Glider Bool!(execute:Exec) -> (then:Exec) {is=title K2Node_MacroInstance}
   > n2 = Set IsFrontend(IsFrontend:bool=IsCosmeticObjectDisplayedInFrontendContext!(CosmeticObject:Object Reference) -> (ReturnValue:bool)) -> (Output_Get:bool)
   > n3 = Branch(Condition=@n2)
   > Set TopSpin_Speed(TopSpin_Speed:double=0) -> (Output_Get:double)
   n4 = ForwardBack_Value_Interp!(
    execute=@n3.else
    Time:double=Get Time()
    Currnet_Value:double=Get TopSpin_Speed()
    Idle_Value:double=1
    Idle_InterpSpeed:double=1
    Forward_Value:double=3
    Forward_InterpSpeed:double=2
    Back_Value:double=-1.75
    Back_InterpSpeed:double=2
    LocalAccelForward:double=Break Fort Anim Input Skydiving(FortAnimInput_Skydiving=Get SkydivingState()) -> (
     ParaGilderRootOffset=~
     ParaGliderRootRotationOverride=~
     bUseParaGlideRootModifier=~
     bIsSkydivingFromLaunchPad=~
     bIsSkydivingFromBus=~
     bIsInVortex=~
     bIsDiving=~
     bIsParachuteOpen=~
     bIsSkydiving=~
     bIsParachuteLeaning=~
     bIsSkydiveLeaning=~
     bIsLeaning=~
     bIsSkydiveDiveMode=~
     bPlaySkydiveDrift=~
     bSkydiveDriftDelayActive=~
     bSkydiveDriftAnimAllowed=~
     bIsGliderRight=~
     bIsGliderCenter=~
     bIsGliderLeft=~
     bIsGliderForward=~
     bIsGliderBack=~
     bHasDeployedAndIsStandingUpright=~
     LocalAccelForward=~
     LocalAccelRight=~
     LocalVelocityRight=~
     SkydiveAimPitch=~
     SkydiveAimYaw=~
     DeployChuteAnimRate=~
     SkydiveDriftAnimRate=~
     SkydiveDriftAnimRateCurrent=~
     SkydiveFidgetAnimRate=~
     SkydiveFidgetAnimRateCurrent=~
     SkydiveAdditiveAlpha=~
     SkydiveDriftDelay=~
    ).LocalAccelForward
   ) -> (then, New_Value:double)
   > Set TopSpin_Speed(TopSpin_Speed=@n4)
   > n5 = For Loop(LastIndex=9)
   > n6 = CalculateFlapRotation!(
    execute
    self
    FlapToCalculate:Name="Get (a ref)"(Array:Array of Names=Get FlapNames() -> (FlapNames:Array of Names), Dimension 1=@n5) -> (Output:"Name (by ref)")
    CurrentRotation:Rotator="Get (a ref)"(Array:Array of Rotators=Get FlapRotationValues(), Dimension 1=@n5) -> (Output:"Rotator (by ref)")
   ) -> (then, NewRotation:Rotator) {is=title K2Node_CallFunction}
   > Set Array Elem!(execute, TargetArray:Array of Rotators=Get FlapRotationValues(), Index:int=@n5, Item:"Rotator (by ref)"=@n6, bSizeToFit:bool=false) -> (then)
   AnimNotify_ClosedEntered() {is=title K2Node_Event}
   n7 = GetOwningComponent()
   > HideBoneByName!(execute, self:Skinned Mesh Component Object Reference=@n7, BoneName:Name=Melee_TwoHanded_Handle, PhysBodyOption:EPhysBodyOp Enum=PBO_None) -> (then)
   AnimNotify_DeployEntered() {is=title K2Node_Event}
   > Delay(Duration=0.05)
   > UnHideBoneByName!(execute, self:Skinned Mesh Component Object Reference=@n7, BoneName:Name=Melee_TwoHanded_Handle) -> (then)
   AnimNotify_TurnONEngines() {is=title K2Node_Event}
   > SetGliderFullyDeployed!(execute, self:Fort Glider Anim Instance Object Reference, bIsFullyDeployed:bool=true) -> (then)
   Event BlueprintInitializeAnimation()
   > Set OwningComponent(OwningComponent:Skeletal Mesh Component Object Reference=GetOwningComponent()) -> (Output_Get:Skeletal Mesh Component Object Reference)
   TryGetPawnOwner()
  ]}
 GetPlayerOpenGliderBool {nodes [
   Inputs()
   > n1 = Branch(Condition=Get IsCosmeticPreview(self=Get OwnerGlider()))
   Outputs!(then:Exec=[@n1.then, @n3.then])
   n2 = Get OwnerFortPawn() -> (OwnerFortPawn:Fort Player Pawn Object Reference)
   n3 = Set bForceOpen(
    execute=@n1.else
    bForceOpen:bool=ShouldForceOpenGlider!(self:Fort Player Animation Interface Object Reference=GetAnimationInterface!(self:Fort Player Pawn Object Reference=@n2) -> (ReturnValue:Fort Player Animation Interface Object Reference)) -> (ReturnValue:bool)
   ) -> (Output_Get:bool)
  ]}
 OnRep_SkydivingState {nodes [OnRep_SkydivingState() {is=title K2Node_FunctionEntry}]}
]
variables [
 FlapDynamicsInterpolationSpeed {type=number, default=3}
 FlapMaxDynamicAngle {type=number, default=60}
 FlapNames {
  type="array<string>"
  default [
   JNT_CanopyFlap_A
   JNT_CanopyFlap_B
   JNT_CanopyFlap_C
   JNT_CanopyFlap_D
   JNT_CanopyFlap_E
   JNT_CanopyFlap_F
   JNT_CanopyFlap_G
   JNT_CanopyFlap_H
   JNT_CanopyFlap_I
   JNT_CanopyFlap_J
  ]
 }
 FlapRotationValues {
  type="array<Rotator>"
  default [<rot 0 0 0>, <rot 0 0 0>, <rot 0 0 0>, <rot 0 0 0>, <rot 0 0 0>, <rot 0 0 0>, <rot 0 0 0>, <rot 0 0 0>, <rot 0 0 0>, <rot 0 0 0>]
 }
 IsFrontend {type=boolean}
 OwningComponent {type=SkeletalMeshComponent}
 SideSpinner_Speed {type=number, default=1}
 Target {type=FortGliderAnimInstance, default=null}
 Time {type=number, default=0.0167}
 TopSpin_Speed {type=number, default=1}
]
